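# Niagara system for a HighTower item aura, drawing six persistent sprites that follow a skeletal mesh.
# A single DecalsSide emitter burst-spawns the particles and partitions them, so each index selects its bone or socket and a position offset from an array.
# Skeletal mesh location is resampled every update from System.SK_Mesh.
# Sprites are pushed by a camera offset, receive dynamic material parameters and are tinted by User.FX_Color.
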NiagaraSystem /HighTower/Items/Aura/NS_HighTower_Aura_fresnel_Converted
dependencies {
 usedDataInterfaces [&NiagaraDataInterfaceArrayFloat3, &NiagaraDataInterfaceSkeletalMesh]
 usedDynamicInputs []
 usedModules [
  &self:SystemSpawnScript.SystemScriptSource.SystemScriptGraph.NiagaraNodeAssignment_0.SetVariables_#1
  &SystemState
 ]
}
emitters {DecalsSide {
  emitterUpdateScript [
   &EmitterState(
    Inactive Response=::NewEnumerator1
    Life Cycle Mode=::NewEnumerator1
    Loop Behavior=::NewEnumerator1
    Loop Duration Mode=::NewEnumerator1
    UseLoopDelay=false
   )
   &SpawnBurst_Instantaneous(Spawn Count=6, Use Spawn Probability=false, UseLoopCountLimit=false)
  ]
  particleSpawnScript [
   &PartitionParticles(Number Of Partitions=6)
   &InitializeParticle(Lifetime=3, Sprite Size Mode=::NewEnumerator3, Uniform Sprite Size=20, UsePositionOffset=false, Write Material Random=true)
   &CameraOffset(Camera Offset Amount=30) {module=CameraOffset001}
   &DynamicMaterialParameters(
    Index 0 Param 1=1.2
    Index 0 Param 2=1
    Module.Index0Param0Write=true
    Module.Index0Param1Write=true
    Module.Index0Param2Write=true
    Module.Index0Param3Write=false
   )
  ]
  particleUpdateScript [
   &ParticleState(Kill Particles When Lifetime Has Elapsed=false, Let Infinitely Lived Particles Die When Emitter Deactivates=true)
   &ScaleColor(
    Scale Mode=::NewEnumerator1
    Scale RGBA=&MakeVector4_FromLinearColor(LinearColor=$User.FX_Color)
    ScaleRGBA=true
   ) {module=ScaleColor001}
   &SkeletalMeshLocation(
    Bone / Socket Index=$Particles.PartitionParticles.Partition
    Bone Sampling Mode=::NewEnumerator4
    MaskSpawnGroup=false
    Sampled Position Offset=&SelectVectorFromArray(
     Direct Array Index=$Particles.PartitionParticles.Partition
     Vector Selection Array=DataInterface() {propertyValues {FloatData [<12 0 0>, <5 0 0>, <0 0 0>, <-12 0 0>, <-5 0 0>, <0 0 0>]}}
    )
    Skeletal Mesh=$System.SK_Mesh
   ) {module=SkeletalMeshLocation001}
  ]
  rendererClasses [&NiagaraSpriteRendererProperties]
  renderers [{rendererClass=&NiagaraSpriteRendererProperties}]
 }}
summary {
 systemName=NS_HighTower_Aura_fresnel_Converted
 emitters [{emitterName=DecalsSide, rendererClasses [&NiagaraSpriteRendererProperties]}]
 userVariables [User.FX_Color {type=LinearColor, description="", defaultValue {struct=LinearColor, value=<rgba 1 1 1 1>}}]
}
